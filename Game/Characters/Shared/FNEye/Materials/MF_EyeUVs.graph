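# Material function for character eyes that fakes corneal refraction of the iris, outputting eye UVs, eye normal, base UVs and debug values.
# A refraction direction from camera, normal and cornea IOR is scaled by iris depth from a displacement texture, then projected onto a tangent basis as a UV offset.
# Sphere masks, aspect ratio and per-eye UV positions handle cornea bounds and left-eye remapping; an iris normal map is also sampled.
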MaterialFunction /Game/Characters/Shared/FNEye/Materials/MF_EyeUVs
graph {nodes [
  n1 = ScalarParameter() {defaultValue=0.02, group=Eye Shader, parameterName=Eye_Refraction Mult, sortPriority=25}
  n2 = Constant2Vector() {g=1, r=-1}
  n3 = NamedRerouteUsage()
  n4 = MaterialFunctionCall() {materialFunction=&TangentBasis}
  n5 = NamedRerouteUsage()
  n6 = DotProduct(A=@n4.X U, B=@n5)
  n7 = Normalize(VectorInput=Subtract(A=@n4.X U, B=Multiply(A=@n6, B=@n5)))
  n8 = DotProduct(A=@n7, B=@n3)
  n9 = DotProduct(A=CrossProduct(A=@n7, B=@n5), B=@n3)
  n10 = Multiply(
   A=Multiply(
    A=Multiply(A=@n1, B=@n2)
    B=MaterialFunctionCall(X=@n8, Y=@n9) {materialFunction=&MakeFloat2}.Result
   )
   B=NamedRerouteUsage()
  )
  Add(A=Add(A=NamedRerouteUsage(), B=@n10), B=@n10)
  CameraVectorWS()
  CameraVectorWS()
  MaterialExpressionConstant_0 = Constant() {r=1.376}
  MaterialExpressionPixelNormalWS_2 = PixelNormalWS()
  MaterialExpressionCameraVectorWS_3 = CameraVectorWS()
  Custom!(internalIoR=@MaterialExpressionConstant_0, normalW=@MaterialExpressionPixelNormalWS_2, cameraW=@MaterialExpressionCameraVectorWS_3) {
   code="float airIoR = 1.00029;\r\n\r\nfloat n = airIoR / internalIoR;\r\n\r\nfloat facing = dot(normalW, cameraW);\r\n\r\nfloat w = n * facing;\r\n\r\nfloat k = sqrt(1+(w-n)*(w+n));\r\n\r\nfloat3 t;\r\nt = (w - k)*normalW - n*cameraW;\r\nt = normalize(t);\r\nreturn -t;"
   desc=Refraction Direction
   inputs [
    {inputName=internalIoR, input {expression=&self:MaterialExpressionConstant_0}}
    {inputName=normalW, input {expression=&self:MaterialExpressionPixelNormalWS_2}}
    {inputName=cameraW, input {expression=&self:MaterialExpressionCameraVectorWS_3}}
   ]
  }
  FunctionOutput(None=NamedRerouteUsage()) {outputName=EyeUVs}
  FunctionOutput(None=NamedRerouteUsage()) {outputName=EyeNormal, sortPriority=1}
  FunctionOutput(None=NamedRerouteUsage()) {outputName=Debug, sortPriority=3}
  FunctionOutput(None=NamedRerouteUsage()) {outputName=UVs, sortPriority=2}
  FunctionOutput(None=NamedRerouteUsage()) {outputName=DebugCornea, sortPriority=4}
  n11 = Distance(A=NamedRerouteUsage(), B=Constant2Vector() {g=0.5, r=0.5})
  n12 = NamedRerouteUsage()
  n13 = Multiply(A=@n12, B=NamedRerouteUsage())
  NamedRerouteDeclaration(None=Add(A=SphereMask(A=@n11, B=@n12) {attenuationRadius=0.0075}, B=SphereMask(A=@n11, B=@n13) {attenuationRadius=0.0075})) {name=DebugCornea, nodeColor=<rgba 0 1 0.7891 1>}
  n14 = TextureCoordinate()
  n15 = NamedRerouteDeclaration(None=Subtract(A=NamedRerouteUsage(), B=NamedRerouteUsage())) {name=LeftEyeUVDelta, nodeColor=<rgba 0.9688 1 0 1>}
  NamedRerouteDeclaration(None=LinearInterpolate(A=@n14, B=Subtract(A=@n14, B=@n15), Alpha=NamedRerouteUsage())) {name=UVs, nodeColor=<rgba 0 1 0.2969 1>}
  n16 = NamedRerouteUsage()
  n17 = Constant2Vector() {g=0.5, r=0.5}
  n18 = NamedRerouteUsage()
  n19 = NamedRerouteUsage()
  NamedRerouteDeclaration(None=Power(Base=SphereMask(A=@n16, B=@n17, Radius=@n18, Hardness=@n19) {attenuationRadius=0.365, hardnessPercent=50}) {constExponent=0.5}) {name=eye displacement mask, nodeColor=<rgba 0.875 1 0 1>}
  n20 = NamedRerouteUsage()
  n21 = DotProduct(A=NamedRerouteUsage(), B=NamedRerouteUsage())
  n22 = LinearInterpolate(Alpha=Multiply(A=@n21, B=@n21)) {constA=0.325}
  NamedRerouteDeclaration(None=Multiply(A=NamedRerouteUsage(), B=Divide(A=@n20, B=@n22) {desc=heightW / cosAlpha - scale distance to iris plane by facing ratio of camera and plane})) {name=Refraction Vector, nodeColor=<rgba 0.0781 1 0 1>}
  n23 = NamedRerouteUsage()
  n24 = NamedRerouteUsage()
  NamedRerouteDeclaration(None=LinearInterpolate(A=@n23, B=Add(A=@n23, B=@n24), Alpha=NamedRerouteUsage())) {name=EyeUVs, nodeColor=<rgba 0 1 0.6016 1>}
  NamedRerouteDeclaration(None=ScalarParameter() {defaultValue=0.078, group=Eye Shader, parameterName="Eye_Cornea Radius (UV)", sliderMax=0.5, sortPriority=10}) {name=EyeCorneaRadius, nodeColor=<rgba 0 1 0.2266 1>}
  NamedRerouteDeclaration(None=VertexNormalWS()) {name=Normal, nodeColor=<rgba 0 0 1 1>}
  NamedRerouteDeclaration(None=Divide(A=Constant() {desc=airIoR, r=1.0003}, B=ScalarParameter() {defaultValue=1.336, group=Eye Shader, parameterName=Eye_Cornea IOR, sliderMax=2}) {constA=1.0003}) {name=n, nodeColor=<rgba 0 1 0.2031 1>}
  n25 = NamedRerouteUsage()
  n26 = NamedRerouteUsage()
  NamedRerouteDeclaration(None=Multiply(A=NamedRerouteUsage(), B=DotProduct(A=@n25, B=@n26) {desc=facing}) {desc=w}) {name=w, nodeColor=<rgba 1 0.8297 0 1>}
  n27 = NamedRerouteUsage()
  n28 = NamedRerouteUsage()
  n29 = Multiply(A=Subtract(A=@n27, B=@n28), B=Add(A=@n27, B=@n28))
  NamedRerouteDeclaration(None=SquareRoot(None=Add(A=@n29)) {desc=k}) {name=k, nodeColor=<rgba 0.518 1 0 1>}
  NamedRerouteDeclaration(None=Transform(None=NamedRerouteUsage())) {name=EyeDirection, nodeColor=<rgba 1 0.3047 0 1>}
  NamedRerouteDeclaration(None=LinearInterpolate(
    A=NamedRerouteUsage()
    B=NamedRerouteUsage()
    Alpha=ScalarParameter() {defaultValue=1, group=Eye Shader, parameterName=Eye_Refraction Mix, sliderMax=1, sortPriority=27}
   )) {name="Eye UVs (normalized)", nodeColor=<rgba 0 1 0.5232 1>}
  NamedRerouteDeclaration(None=Constant() {r=0.365}) {name=IrisSize, nodeColor=<rgba 1 0 0.7969 1>}
  n30 = Multiply(A=NamedRerouteUsage(), B=NamedRerouteUsage())
  n31 = NamedRerouteUsage()
  n32 = Multiply(A=NamedRerouteUsage(), B=NamedRerouteUsage())
  n33 = Add(
   A=Subtract(A=Divide(A=@n30, B=@n31)) {constB=-0.5}
   B=Divide(A=Multiply(A=@n32) {constB=-1, desc=compensates for normal map not being 0-1}, B=@n31)
  )
  n34 = TextureSampleParameter2D(UVs=NamedRerouteDeclaration(None=Saturate(None=@n33)) {name=Iris UVs, nodeColor=<rgba 0.2422 1 0 1>}) {
   group=Eye Shader
   parameterName=Eye_Iris Normal
   samplerType=SAMPLERTYPE_Normal
   sortPriority=20
   texture=/Game/Characters/Shared/FNEye/Textures/T_Eye_Sphere_N
  }
  NamedRerouteDeclaration(None=@n34.RGB) {name="eyeNormal (unflattened)", nodeColor=<rgba 0 1 0.1797 1>}
  NamedRerouteDeclaration(None=ScalarParameter() {defaultValue=0.2, group=Eye Shader, parameterName=Eye_Cornea Mask Hardness, sliderMax=1, sortPriority=11}) {name=Eye Cornea Mask Hardness Reroute, nodeColor=<rgba 0.8047 0 1 1>}
  n35 = TextureSample(UVs=NamedRerouteUsage()) {
   constCoordinate=1
   desc=UVs not scaled by center as the base UVs are a ortho projection and this depth needs to stay in that space
   samplerType=SAMPLERTYPE_LinearColor
   texture=/Game/Characters/Shared/FNEye/Textures/T_EyeMidPlaneDisplacement
  }
  n36 = ScalarParameter() {defaultValue=0.4, group=Eye Shader, parameterName=Eye_Iris UV Radius, sliderMax=0.5}
  n37 = TextureSample(UVs=MaterialFunctionCall(X=Add(A=@n36) {constB=0.5}, Y=Constant() {r=0.5}) {materialFunction=&MakeFloat2}.Result) {
   desc="Sample Displacement map as above, sampled at the Iris Radius"
   samplerType=SAMPLERTYPE_LinearColor
   texture=&T_EyeMidPlaneDisplacement
  }
  n38 = Subtract(A=@n35.RGB, B=@n37.R)
  n39 = Constant()
  NamedRerouteDeclaration(None=Multiply(A=Max(A=@n38, B=@n39), B=ScalarParameter() {defaultValue=0.5, group=Eye Shader, parameterName=Eye_Iris Depth Scale, sliderMax=1}) {desc=Iris Depth}) {name=height, nodeColor=<rgba 0 1 0.4375 1>}
  n40 = NamedRerouteUsage()
  n41 = NamedRerouteUsage()
  n42 = TextureCoordinate()
  n43 = NamedRerouteUsage()
  NamedRerouteDeclaration(None=SphereMask(A=Multiply(A=@n40, B=@n41), B=Multiply(A=@n42, B=@n41) {constB=2}, Radius=Multiply(A=@n43) {constB=0.5}) {attenuationRadius=0.1}) {name=LeftEyeMask, nodeColor=<rgba 1 0 0.5391 1>}
  n44 = VectorParameter() {
   desc="uvPosition (UV0) of the left Eye. Top left is 0,0\r\nSet it to the same as the right eye if the same texture is used for both"
   group=Eye Shader
   parameterName="Eye_Left UV Position (UV0)"
   sortPriority=13
   defaultValue=<rgba 0.94 0.35 0 1>
  }
  NamedRerouteDeclaration(None=ComponentMask(None=@n44.RGB) {g=true, r=true}) {name=Left Eye UVPosition, nodeColor=<rgba 1 0.375 0 1>}
  n45 = ScalarParameter() {
   defaultValue=2
   desc="Texture Width/Height (2048x1024 would be 2)"
   group=Eye Shader
   parameterName=Eye_Texture AspectRatio
   sliderMax=4
   sortPriority=1
  }
  NamedRerouteDeclaration(None=MaterialFunctionCall(X=Constant() {r=1}, Y=Divide(B=@n45)) {materialFunction=&MakeFloat2}.Result) {name=AspectRatio, nodeColor=<rgba 1 0.4688 0 1>}
  n46 = VectorParameter() {group=Eye Shader, parameterName="Eye_Right UV Position (UV0)", sortPriority=12, defaultValue=<rgba 0.94 0.122 0 0>}
  NamedRerouteDeclaration(None=ComponentMask(None=@n46.RGB) {g=true, r=true}) {name=RightEyeUVPosition, nodeColor=<rgba 0.7812 1 0 1>}
  NamedRerouteDeclaration(None=AppendVector(A=@n10, B=Constant())) {name=debug, nodeColor=<rgba 1 0.4688 0 1>}
  NamedRerouteDeclaration(None=Add(A=@n10, B=NamedRerouteUsage())) {name=Refracted UVs, nodeColor=<rgba 0.6406 1 0 1>}
  n47 = WorldPosition() {worldPositionShaderOffset=WPT_CameraRelativeNoOffsets}
  NamedRerouteDeclaration(None=Normalize(VectorInput=Multiply(A=@n47.XYZ) {constB=-1})) {name=Camera Vector, nodeColor=<rgba 1 0 1 1>}
  n48 = NamedRerouteUsage()
  n49 = NamedRerouteUsage()
  n50 = Subtract(A=Multiply(A=Subtract(A=@n48, B=@n49), B=NamedRerouteUsage()), B=Multiply(A=NamedRerouteUsage(), B=NamedRerouteUsage())) {desc=t}
  NamedRerouteDeclaration(None=Multiply(A=Normalize(VectorInput=@n50)) {constB=-1}) {name=Eye Refraction Direction, nodeColor=<rgba 0 0.2882 1 1>}
  PixelNormalWS()
  StaticSwitch(True=VertexNormalWS(), False=PixelNormalWS(), Value=StaticBoolParameter() {group=Eye Shader, parameterName=Eye_UseVertexAttributes})
 ]}
