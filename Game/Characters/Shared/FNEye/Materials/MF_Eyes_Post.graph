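# Material function that post-processes Fortnite character eyes, turning incoming attributes or eye textures into final shading attributes.
# Base color supports tinting and debug views, roughness and metallic are adjustable and swizzlable, and flattened normals blend with an eye normal.
# Emissive combines a maskable facing or fresnel glow, clamped on mobile, with a UV sphere-mask, sun or camera-vector highlight.
# Blend scalars fade each modified channel except specular against its unmodified value.
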MaterialFunction /Game/Characters/Shared/FNEye/Materials/MF_Eyes_Post
graph {nodes [
  n1 = NamedRerouteUsage()
  n2 = NamedRerouteUsage()
  n3 = SetMaterialAttributes!(
   MaterialAttributes
   BaseColor=StaticSwitchParameter(
    True=NamedRerouteUsage()
    False=StaticSwitchParameter(True=@n1, False=@n2) {group=Eye Shader, parameterName=Eye_Visualize Cornea, sortPriority=2}
   ) {group=Eye Shader, parameterName=Eye_Debug, sortPriority=2}
   Specular=NamedRerouteUsage()
   Metallic=NamedRerouteUsage()
   Roughness=NamedRerouteUsage()
   Normal=NamedRerouteUsage()
   ShadingModel=ShadingModel()
   EmissiveColor=NamedRerouteUsage()
   Anisotropy
   Opacity
  ) {attributeSetTypes [
    #1
    #2
    #3
    #4
    #5
    #6
    #7
    #8
    #9
   ]}
  FunctionOutput(None=@n3)
  n4 = NamedRerouteUsage()
  n5 = StaticSwitchParameter(True=NamedRerouteUsage(), False=NamedRerouteUsage()) {group=Eye Shader, parameterName=Eye_ Use Sun Highlight, sortPriority=35}
  NamedRerouteDeclaration(None=Multiply(A=StaticSwitchParameter(True=@n4, False=@n5) {defaultValue=true, group=Eye Shader, parameterName=Eye_Use UV Highlight}) {constB=0.6}) {name=EmissiveHighlight, nodeColor=<rgba 0.5938 0 1 1>}
  n6 = NamedRerouteUsage()
  n7 = NamedRerouteUsage()
  n8 = ComponentMask(None=VectorParameter() {group=Eye Shader, parameterName=Eye_UV Highlight Pos, defaultValue=<rgba 0.92 0.12 0 0>}.RGB) {g=true, r=true}
  n9 = NamedRerouteUsage()
  NamedRerouteDeclaration(None=SphereMask(
    A=Multiply(A=@n6, B=@n7)
    B=Multiply(A=@n8, B=@n9)
    Radius=ScalarParameter() {defaultValue=0.005, group=Eye Shader, parameterName=Eye_UV Highlight Size}
    Hardness=ScalarParameter() {defaultValue=0.6, group=Eye Shader, parameterName=Eye_UV Highlight Hardness, sliderMax=1}
   )) {name=uvHighlight, nodeColor=<rgba 0.7813 0 1 1>}
  NamedRerouteDeclaration(None=FunctionInput() {inputName=EyeUVs, inputType=FunctionInput_Vector2}) {name=EyeUVs, nodeColor=<rgba 0 1 0.6016 1>}
  n10 = CollectionParameter() {
   collection=&FortniteMaterialParameters
   parameterName=SunlightDirectionalVector
  }
  n11 = Normalize(VectorInput=Add(A=CameraVectorWS(), B=ComponentMask(None=@n10) {b=true, g=true, r=true}))
  n12 = Normalize(VectorInput=VertexNormalWS())
  NamedRerouteDeclaration(None=SmoothStep(Value=DotProduct(A=@n11, B=@n12) {desc=Blinn}) {constMin=0.995}) {name=SunDirectionHighlight, nodeColor=<rgba 1 0.75 0 1>}
  n13 = NamedRerouteUsage()
  n14 = NamedRerouteUsage()
  n15 = TextureSampleParameter2D(UVs=@n14) {group=Eye Shader, parameterName=Eye_Emissive, texture=&T_BlacksRGB}
  n16 = NamedRerouteUsage()
  n17 = StaticComponentMaskParameter(None=NamedRerouteDeclaration(None=StaticSwitch(True=@n13, False=@n15.RGB, Value=@n16)) {name=Emissive_Raw, nodeColor=<rgba 1 0 0.7266 1>}) {defaultB=true, defaultG=true, defaultR=true, group=Eye Shader, parameterName=Eye_Emissive Component Mask}
  n18 = StaticSwitchParameter(True=Multiply(A=@n17, B=@n17), False=@n17) {group=Eye Shader, parameterName=Eye_LinearEmissiveMask}
  n19 = Fresnel(Normal=Transform(None=NamedRerouteUsage())) {baseReflectFraction=0, exponent=1}
  n20 = ScalarParameter() {
   defaultValue=1
   desc=Set to 1 to have fresnel emissive instead of camera facing emissive
   group=Eye Shader
   parameterName=Eye_CameraFacing_Inverted
   sliderMax=1
  }
  n21 = Round(None=@n20)
  n22 = Saturate(None=Abs(None=Subtract(A=@n19, B=@n21)))
  n23 = ScalarParameter() {defaultValue=1, group=Eye Shader, parameterName=Eye_Fresnel_Power, sliderMax=10}
  n24 = Multiply(A=@n18, B=Clamp(None=Power(Base=@n22, Exp=@n23)) {maxDefault=100000}) {desc=Emis*Facing}
  n25 = ScalarParameter() {defaultValue=1, group=Eye Shader, parameterName=Eye_Emissive Mult}
  n26 = Clamp(None=@n25, Max=ScalarParameter() {defaultValue=1, group=Eye Shader, parameterName=Eye_Emissive Mult_MobileMax})
  NamedRerouteDeclaration(None=Multiply(
    A=StaticSwitchParameter(True=@n24, False=@n18) {group=Eye Shader, parameterName=Eye_CameraFacing}
    B=ShadingPathSwitch(Default=@n25, Deferred=@n25, Mobile=@n26) {desc=Migrated from FeatureLevelSwitch}
   )) {name=EyeEmissive, nodeColor=<rgba 1 0 0.6328 1>}
  NamedRerouteDeclaration(None=FunctionInput() {inputName=DebugCornea, inputType=FunctionInput_Vector2, sortPriority=4}) {name=DebugCornea, nodeColor=<rgba 0.2422 1 0 1>}
  NamedRerouteDeclaration(None=LinearInterpolate(A=NamedRerouteUsage(), B=Constant3Vector() {constant=<rgba 1 0 0 0>}, Alpha=NamedRerouteUsage())) {name=VisualizeCornea, nodeColor=<rgba 1 0.3281 0 1>}
  n27 = Normalize(VectorInput=VertexNormalWS())
  n28 = VectorParameter() {group=Eye Shader, parameterName=Eye_Camera Light Vector, sortPriority=36, defaultValue=<rgba -0.2 0.075 -0.5 0>}
  n29 = Transform(None=Normalize(VectorInput=@n28.RGB)) {transformSourceType=TRANSFORMSOURCE_View}
  NamedRerouteDeclaration(None=SmoothStep(Value=DotProduct(A=@n27, B=@n29)) {constMin=0.995}) {name=Camera Highlight, nodeColor=<rgba 1 0.75 0 1>}
  n30 = ScalarParameter() {defaultValue=2, desc="Texture Width/Height (2048x1024 would be 2)", group=Eye Shader, parameterName=Eye_Texture AspectRatio, sortPriority=1}
  NamedRerouteDeclaration(None=MaterialFunctionCall(X=Constant() {r=1}, Y=Divide(B=@n30)) {materialFunction=&MakeFloat2}.Result) {name=AspectRatio, nodeColor=<rgba 1 0.4688 0 1>}
  n31 = MaterialFunctionCall(
   Normal=NamedRerouteUsage()
   Flatness=ScalarParameter() {defaultValue=0.25, group=Eye Shader, parameterName=Eye_Iris Normal Flatten, sortPriority=21}
  ) {materialFunction=&FlattenNormal}
  n32 = NamedRerouteUsage()
  NamedRerouteDeclaration(None=LinearInterpolate(
    A=NamedRerouteUsage()
    B=MaterialFunctionCall(BaseNormal=@n31.Result, AdditionalNormal=@n32) {materialFunction=&BlendAngleCorrectedNormals}.Result
    Alpha=ScalarParameter() {defaultValue=1, desc=Blend modified normal with original normal, group=Eye Shader, parameterName=Eye_Blend_Normal}
   )) {name=Normal_OUT, nodeColor=<rgba 0 1 0.9063 1>}
  n33 = TextureSampleParameter2D() {
   group=Eye Shader
   parameterName=Eye_Normal
   samplerType=SAMPLERTYPE_Normal
   sortPriority=0
   texture=&FlatNormalBlue
  }
  NamedRerouteDeclaration(None=StaticSwitch(True=NamedRerouteUsage(), False=@n33.RGB, Value=NamedRerouteUsage())) {name=Normal_Raw, nodeColor=<rgba 0.1435 0.3013 1 1>}
  n34 = NamedRerouteUsage()
  NamedRerouteDeclaration(None=StaticSwitch(
    True=NamedRerouteUsage()
    False=TextureSampleParameter2D(UVs=@n34) {group=Eye Shader, parameterName=Eye_Diffuse, sortPriority=0, texture=/Game/Characters/Shared/FNEye/Textures/T_Eye_D}.RGB
    Value=NamedRerouteUsage()
   )) {name=BaseColor_Raw, nodeColor=<rgba 0 0.8744 1 1>}
  NamedRerouteDeclaration(None=StaticBoolParameter() {group=Eye Shader, parameterName=Eye_UseMaterialAttributes}) {name=UseMaterialAttributes, nodeColor=<rgba 0.8281 1 0 1>}
  NamedRerouteDeclaration(None=FunctionInput() {inputName=Debug, sortPriority=3}) {name=debug, nodeColor=<rgba 1 0.4688 0 1>}
  n35 = GetMaterialAttributes(None=FunctionInput() {inputName=MA_IN, inputType=FunctionInput_MaterialAttributes, sortPriority=-2}) {attributeGetTypes [
    #5
    #1
    #2
    #3
    #4
    #7
   ]}
  NamedRerouteDeclaration(None=@n35.Normal) {name=Normal_IN, nodeColor=<rgba 0 0.2969 1 1>}
  NamedRerouteDeclaration(None=@n35.BaseColor) {name=BaseColor_IN, nodeColor=<rgba 0.6719 0.0633 0.6678 1>}
  NamedRerouteDeclaration(None=@n35.Specular) {name=Specular_IN, nodeColor=<rgba 0.4323 0.2432 0.4189 1>}
  NamedRerouteDeclaration(None=@n35.Metallic) {name=Metallic_IN, nodeColor=<rgba 0.1719 1 0 1>}
  NamedRerouteDeclaration(None=@n35.Roughness) {name=Roughness_IN, nodeColor=<rgba 0 0.4375 1 1>}
  NamedRerouteDeclaration(None=LinearInterpolate(A=NamedRerouteUsage(), B=NamedRerouteUsage(), Alpha=ScalarParameter() {defaultValue=1, group=Eye Shader, parameterName=Eye_Blend_BaseColor})) {name=BaseColor_OUT, nodeColor=<rgba 0 0.4141 1 1>}
  n36 = Desaturation(None=NamedRerouteUsage(), Fraction=Constant() {r=1})
  n37 = Multiply(A=@n36, B=VectorParameter() {group=Eye Color Tinting, parameterName=Eye_TintColor, defaultValue=<rgba 0.53 1 0.47 1>}.RGB)
  n38 = SphereMask(
   A=NamedRerouteUsage()
   B=Constant() {r=0.5}
   Radius=ScalarParameter() {defaultValue=0.4, group=Eye Color Tinting, parameterName=Eye_TintMask_Radius, sliderMax=1}
  ) {attenuationRadius=0.4, hardnessPercent=80}
  NamedRerouteDeclaration(None=StaticSwitchParameter(True=LinearInterpolate(A=@n36, B=@n37, Alpha=@n38), False=NamedRerouteUsage()) {group=Eye Color Tinting, parameterName=UseEyeColorTinting, sortPriority=1}) {name=BaseColor_Modified, nodeColor=<rgba 0 0.3437 1 1>}
  n39 = TextureSampleParameter2D(UVs=NamedRerouteUsage()) {
   group=Eye Shader
   parameterName=Eye_SpecularMasks
   samplerType=SAMPLERTYPE_Masks
   sortPriority=0
   texture=/Game/Characters/Shared/FNEye/Textures/T_Eye_S
   channelNames {a="", b=Roughness, g=Metallic, r=Specular}
  }
  NamedRerouteDeclaration(None=StaticSwitch(True=NamedRerouteUsage(), False=@n39.Specular, Value=NamedRerouteUsage())) {name=Specular_OUT, nodeColor=<rgba 1 0 0.9844 1>}
  n40 = StaticSwitch(True=NamedRerouteUsage(), False=@n39.Roughness, Value=NamedRerouteUsage())
  n41 = StaticSwitch(True=NamedRerouteUsage(), False=@n39.Metallic, Value=NamedRerouteUsage())
  n42 = StaticBoolParameter() {group=Eye Shader, parameterName=Eye_SwizzleRoughnessToGreen}
  n43 = StaticSwitch(True=@n40, False=@n41, Value=@n42)
  n44 = ScalarParameter() {defaultValue=0.1, group=Eye Shader, parameterName=Eye_MetallicMult, sortPriority=2}
  NamedRerouteDeclaration(None=LinearInterpolate(A=@n43, B=Multiply(A=@n43, B=@n44) {constB=0}, Alpha=ScalarParameter() {defaultValue=1, group=Eye Shader, parameterName=Eye_Blend_Metallic})) {name=Metallic_OUT, nodeColor=<rgba 0.2422 0 1 1>}
  n45 = StaticSwitch(True=@n41, False=@n40, Value=@n42)
  n46 = ScalarParameter() {defaultValue=0.3, group=Eye Shader, parameterName=Eye_RoughnessMin, sortPriority=2}
  NamedRerouteDeclaration(None=LinearInterpolate(A=@n45, B=Max(A=@n45, B=@n46), Alpha=ScalarParameter() {defaultValue=1, group=Eye Shader, parameterName=Eye_Blend_Roughness})) {name=Roughness_OUT, nodeColor=<rgba 0.2891 1 0 1>}
  n47 = NamedRerouteUsage()
  n48 = NamedRerouteUsage()
  NamedRerouteDeclaration(None=LinearInterpolate(A=NamedRerouteUsage(), B=Max(A=@n47, B=@n48), Alpha=ScalarParameter() {defaultValue=1, group=Eye Shader, parameterName=Eye_Blend_Emissive})) {name=Emissive_OUT, nodeColor=<rgba 0 1 0.7656 1>}
  NamedRerouteDeclaration(None=@n35.EmissiveColor) {name=Emissive_IN, nodeColor=<rgba 1 0 0.4453 1>}
  NamedRerouteDeclaration(None=FunctionInput() {inputName=UVs, inputType=FunctionInput_Vector2, sortPriority=2}) {name=UVs, nodeColor=<rgba 0 0.1562 1 1>}
  NamedRerouteDeclaration(None=FunctionInput() {inputName=EyeNormal, sortPriority=1}) {name=eyeNormal, nodeColor=<rgba 0.4766 1 0 1>}
  PixelNormalWS()
  n49 = VectorParameter() {group=Eye Shader, parameterName=Eye_EmissiveUVs_RG_UpperLeftCorner_BA_LowerRightCorner, defaultValue=<rgba 0 0 1 1>}
  StaticSwitchParameter(
   True=MaterialFunctionCall!(
    UVs=@n14
    TextureIn=TextureObjectParameter() {group=Eye Shader, parameterName=Eye_Emissive, texture=&T_BlacksRGB}
    UpperLeftCorner=ComponentMask(None=@n49.RGB) {g=true, r=true}
    LowerRightCorner=AppendVector(A=@n49.B, B=@n49.B)
    ExportFloat4
   ) {materialFunction=/Engine/Functions/Engine_MaterialFunctions01/Texturing/TextureCropping}.CroppedMasked
   False=@n17
  ) {desc=CROP_EMIS, group=Eye Shader, parameterName=Eye_CroppedEmissive}
 ]}
