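# Toon eye material for skeletal-mesh characters, built from shared material functions with material attributes and Default Lit shading.
# MF_EyeUVs feeds eye UVs into the BaseFN character PreFX function, whose result goes through MF_AnimeShading for the anime look, then into PostFX.
# An MF_Eyes_Post call also takes the PreFX result and eye UVs, but its output does not feed the anime shading chain.
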Material /Game/Characters/Shared/FNEye/Materials/M_Eyes_Toon
graph {
 properties {
  bAutomaticallySetUsageInEditor=false
  bUseMaterialAttributes=true
  bUsedWithMeshDeformer=true
  bUsedWithSkeletalMesh=true
  pixelDepthOffsetMode=PDOM_Legacy
  usedShadingModels=Default Lit
 }
 nodes [
  n1 = MaterialFunctionCall() {materialFunction=&/MF_EyeUVs}
  n2 = MaterialFunctionCall(TexCoords=@n1.EyeUVs) {materialFunction=&BaseFN_CharacterFunction_PreFX}
  MaterialFunctionCall!(MA_IN=@n2.Result, EyeUVs=@n1.EyeUVs, EyeNormal=@n1.EyeUVs, UVs=@n1.EyeUVs, Debug=@n1.EyeUVs, DebugCornea=@n1.EyeUVs) {materialFunction=&/MF_Eyes_Post}
  MaterialFunctionCall(
   AO=@n2.AO
   Distance Blend=@n2.AO
   Fresnel=@n2.AO
   Gmap=@n2.AO
   Pre Skinned Local Position=@n2.AO
   Material=MaterialFunctionCall!(
    Material=@n2.Result
    LitDiffuseOverride
    SSC_AlphaOverride
    Tint Mask
    DF InkLine Visibility Mask
    DF Inkline Injection
    DF Inkline Color Override
    DF Inkline Color Override Mask
    CustomUVs=@n1.EyeUVs
   ) {materialFunction=/Game/Packages/Fortress_SharedMaterials/Anime_Master_Material/MF_AnimeShading}.Result
  ) {materialFunction=&BaseFN_CharacterFunction_PostFX}
 ]
}
